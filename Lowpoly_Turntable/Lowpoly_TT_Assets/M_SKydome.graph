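# M_SKydome is an unlit sky material for the skydome in the low-poly turntable assets, flagged as sky.
# A single SkyColor vector parameter, defaulting to a muted lavender blue, supplies a flat color.
# A ReflectionCapturePassSwitch node takes that same color on both its default and reflection inputs, with no output connection shown.
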Material /Lowpoly_Turntable/Lowpoly_TT_Assets/M_SKydome
graph {
 properties {bIsSky=true, pixelDepthOffsetMode=PDOM_Legacy, shadingModel=MSM_Unlit, usedShadingModels=Unlit}
 nodes [
  n1 = VectorParameter() {parameterName=SkyColor, defaultValue=<rgba 0.3012 0.2806 0.5 0>}
  ReflectionCapturePassSwitch!(Default=@n1.RGB, Reflection=@n1.RGB)
 ]
}
