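# Level Sequence for a low-poly model turntable, presumably for model checks.
# Turntable_UserObjects spins about Z in full turns that speed up, keyed across two transform sections.
# Trigger events on BP_Turntable_ModelCheck fire as turns complete and at later intervals, alongside an Object Num track; their handlers are not shown.
# Three camera cuts split the shot, with two identically placed cine cameras, two regular cameras and a fixed background mesh bound.
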LevelSequence /Lowpoly_Turntable/Lowpoly_TT_Assets/Turntable_Sequence_Lowpoly
bindings [
 Cine Camera Actor {
  id=#1
  possessed_class=CineCameraActor
  tracks [
   {
    type=MovieScene3DTransformTrack
    sections [{
      type=MovieScene3DTransformSection
      channels [
       Location.X {type=MovieSceneScriptingDoubleChannel, default=364.2415}
       Location.Y {type=MovieSceneScriptingDoubleChannel, default=362.644}
       Location.Z {type=MovieSceneScriptingDoubleChannel, default=134.752}
       Rotation.X {type=MovieSceneScriptingDoubleChannel, default=0}
       Rotation.Y {type=MovieSceneScriptingDoubleChannel, default=-0.6985}
       Rotation.Z {type=MovieSceneScriptingDoubleChannel, default=-135.9996}
       Scale.X {type=MovieSceneScriptingDoubleChannel, default=1}
       Scale.Y {type=MovieSceneScriptingDoubleChannel, default=1}
       Scale.Z {type=MovieSceneScriptingDoubleChannel, default=1}
      ]
     }]
   }
   {type=MovieSceneSpawnTrack, sections [{type=MovieSceneSpawnSection, channels [None {type=MovieSceneScriptingBoolChannel, default=true}]}]}
  ]
 }
 CameraComponent {
  id=#2
  parent_id=#3
  possessed_class=CineCameraComponent
  tracks [
   {
    type=MovieSceneFloatTrack
    display_name=Current Focal Length
    sections [{type=MovieSceneFloatSection, channels [None {type=MovieSceneScriptingFloatChannel, default=35}]}]
   }
   {
    type=MovieSceneFloatTrack
    display_name="Manual Focus Distance (Focus Settings)"
    sections [{type=MovieSceneFloatSection, channels [None {type=MovieSceneScriptingFloatChannel, default=100000}]}]
   }
   {
    type=MovieSceneFloatTrack
    display_name=Current Aperture
    sections [{type=MovieSceneFloatSection, channels [None {type=MovieSceneScriptingFloatChannel, default=2.8}]}]
   }
  ]
 }
 "Cine Camera Actor (2)" {
  id=#3
  possessed_class=CineCameraActor
  tracks [
   {
    type=MovieScene3DTransformTrack
    sections [{
      type=MovieScene3DTransformSection
      channels [
       Location.X {type=MovieSceneScriptingDoubleChannel, default=364.2415}
       Location.Y {type=MovieSceneScriptingDoubleChannel, default=362.644}
       Location.Z {type=MovieSceneScriptingDoubleChannel, default=134.752}
       Rotation.X {type=MovieSceneScriptingDoubleChannel, default=0}
       Rotation.Y {type=MovieSceneScriptingDoubleChannel, default=-0.6985}
       Rotation.Z {type=MovieSceneScriptingDoubleChannel, default=-135.9996}
       Scale.X {type=MovieSceneScriptingDoubleChannel, default=1}
       Scale.Y {type=MovieSceneScriptingDoubleChannel, default=1}
       Scale.Z {type=MovieSceneScriptingDoubleChannel, default=1}
      ]
     }]
   }
   {type=MovieSceneSpawnTrack, sections [{type=MovieSceneSpawnSection, channels [None {type=MovieSceneScriptingBoolChannel, default=true}]}]}
  ]
 }
 Turntable_UserObjects {
  id=#4
  possessed_class=Actor
  tracks [{
    type=MovieScene3DTransformTrack
    sections [
     {
      type=MovieScene3DTransformSection
      channels [
       Location.X {type=MovieSceneScriptingDoubleChannel, default=0}
       Location.Y {type=MovieSceneScriptingDoubleChannel, default=0}
       Location.Z {type=MovieSceneScriptingDoubleChannel, default=0}
       Rotation.X {type=MovieSceneScriptingDoubleChannel, default=0}
       Rotation.Y {type=MovieSceneScriptingDoubleChannel, default=0}
       Rotation.Z {type=MovieSceneScriptingDoubleChannel, default=1710, keys [[0.755,-90],[90.2063,270],[112.7337,630],[135.01,990],[155,1710]]}
       Scale.X {type=MovieSceneScriptingDoubleChannel, default=1}
       Scale.Y {type=MovieSceneScriptingDoubleChannel, default=1}
       Scale.Z {type=MovieSceneScriptingDoubleChannel, default=1}
      ]
     }
     {
      type=MovieScene3DTransformSection
      channels [
       Location.X {type=MovieSceneScriptingDoubleChannel, default=0}
       Location.Y {type=MovieSceneScriptingDoubleChannel, default=0}
       Location.Z {type=MovieSceneScriptingDoubleChannel, default=0}
       Rotation.X {type=MovieSceneScriptingDoubleChannel, default=0}
       Rotation.Y {type=MovieSceneScriptingDoubleChannel, default=0}
       Rotation.Z {type=MovieSceneScriptingDoubleChannel, default=1350, keys [[0,-90],[90.4013,270],[113,630],[135,990]]}
       Scale.X {type=MovieSceneScriptingDoubleChannel, default=1}
       Scale.Y {type=MovieSceneScriptingDoubleChannel, default=1}
       Scale.Z {type=MovieSceneScriptingDoubleChannel, default=1}
      ]
     }
    ]
   }]
 }
 CameraComponent {
  id=#5
  parent_id=#6
  possessed_class=CameraComponent
  tracks [{
    type=MovieSceneFloatTrack
    display_name=Field Of View
    sections [{type=MovieSceneFloatSection, channels [None {type=MovieSceneScriptingFloatChannel, default=35}]}]
   }]
 }
 SM_background {
  id=#7
  possessed_class=StaticMeshActor
  tracks [{
    type=MovieScene3DTransformTrack
    sections [{
      type=MovieScene3DTransformSection
      channels [
       Location.X {type=MovieSceneScriptingDoubleChannel, default=0, const [0,1]}
       Location.Y {type=MovieSceneScriptingDoubleChannel, default=0, const [-460,1]}
       Location.Z {type=MovieSceneScriptingDoubleChannel, default=0, const [202,1]}
       Rotation.X {type=MovieSceneScriptingDoubleChannel, default=0, const [0,1]}
       Rotation.Y {type=MovieSceneScriptingDoubleChannel, default=0, const [0,1]}
       Rotation.Z {type=MovieSceneScriptingDoubleChannel, default=0, const [0,1]}
       Scale.X {type=MovieSceneScriptingDoubleChannel, default=1, const [1,1]}
       Scale.Y {type=MovieSceneScriptingDoubleChannel, default=1, const [1,1]}
       Scale.Z {type=MovieSceneScriptingDoubleChannel, default=1, const [1,1]}
      ]
     }]
   }]
 }
 CameraActor2 {
  id=#6
  possessed_class=CameraActor
  tracks [{
    type=MovieScene3DTransformTrack
    sections [{
      type=MovieScene3DTransformSection
      channels [
       Location.X {type=MovieSceneScriptingDoubleChannel, default=0}
       Location.Y {type=MovieSceneScriptingDoubleChannel, default=0}
       Location.Z {type=MovieSceneScriptingDoubleChannel, default=-111.6854}
       Rotation.X {type=MovieSceneScriptingDoubleChannel, default=0, const [0,1]}
       Rotation.Y {type=MovieSceneScriptingDoubleChannel, default=90, const [90,1]}
       Rotation.Z {type=MovieSceneScriptingDoubleChannel, default=0, const [0,1]}
       Scale.X {type=MovieSceneScriptingDoubleChannel, default=1}
       Scale.Y {type=MovieSceneScriptingDoubleChannel, default=1}
       Scale.Z {type=MovieSceneScriptingDoubleChannel, default=1}
      ]
     }]
   }]
 }
 CameraActor {id=#8, possessed_class=CameraActor, tracks []}
 CameraComponent {
  id=#9
  parent_id=#1
  possessed_class=CineCameraComponent
  tracks [
   {
    type=MovieSceneFloatTrack
    display_name=Current Focal Length
    sections [{type=MovieSceneFloatSection, channels [None {type=MovieSceneScriptingFloatChannel, default=35}]}]
   }
   {
    type=MovieSceneFloatTrack
    display_name="Manual Focus Distance (Focus Settings)"
    sections [{type=MovieSceneFloatSection, channels [None {type=MovieSceneScriptingFloatChannel, default=100000}]}]
   }
   {
    type=MovieSceneFloatTrack
    display_name=Current Aperture
    sections [{type=MovieSceneFloatSection, channels [None {type=MovieSceneScriptingFloatChannel, default=2.8}]}]
   }
  ]
 }
 CameraComponent {
  id=#10
  parent_id=#8
  possessed_class=CameraComponent
  tracks [{
    type=MovieSceneFloatTrack
    display_name=Field Of View
    sections [{type=MovieSceneFloatSection, channels [None {type=MovieSceneScriptingFloatChannel, default=35}]}]
   }]
 }
 BP_Turntable_ModelCheck {
  id=#11
  possessed_class=BP_Turntable_Lowpoly_C
  tracks [
   {
    type=MovieSceneEventTrack
    sections [{
      type=MovieSceneEventTriggerSection
      channels [None {
        type=MovieSceneScriptingEventChannel
        keys [
         [0, "<Struct 'MovieSceneEvent' (0x851369ac8a0) {}>"]
         [90, "<Struct 'MovieSceneEvent' (0x851369ac800) {}>"]
         [113, "<Struct 'MovieSceneEvent' (0x851369ab9a0) {}>"]
         [135, "<Struct 'MovieSceneEvent' (0x851369ac9e0) {}>"]
         [155, "<Struct 'MovieSceneEvent' (0x851369ac580) {}>"]
         [170, "<Struct 'MovieSceneEvent' (0x851369ac440) {}>"]
         [185, "<Struct 'MovieSceneEvent' (0x851369ac4e0) {}>"]
         [200, "<Struct 'MovieSceneEvent' (0x851369ac260) {}>"]
        ]
       }]
     }]
   }
   {
    type=MovieSceneFloatTrack
    display_name=Object Num
    sections [{type=MovieSceneFloatSection, channels [None {type=MovieSceneScriptingFloatChannel, default=0, const [360.4546,1]}]}]
   }
   {
    type=MovieSceneEventTrack
    sections [{
      type=MovieSceneEventTriggerSection
      channels [None {type=MovieSceneScriptingEventChannel, const ["<Struct 'MovieSceneEvent' (0x851369ac120) {}>", 1]}]
     }]
   }
  ]
 }
]
display_rate {denominator=1, numerator=30}
tick_resolution {denominator=1, numerator=24000}
tracks [{
  type=MovieSceneCameraCutTrack
  sections [
   {type=MovieSceneCameraCutSection, end_frame=155, start_frame=0}
   {type=MovieSceneCameraCutSection, end_frame=200, start_frame=155}
   {type=MovieSceneCameraCutSection, end_frame=750, start_frame=200}
  ]
 }]
